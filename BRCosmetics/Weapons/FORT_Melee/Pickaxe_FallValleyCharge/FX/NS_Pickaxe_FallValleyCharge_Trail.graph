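# Niagara weapon trail for the FallValleyCharge pickaxe, drawn as a single short-lived ribbon emitter, Ribbon_Side.
# Shared FN AnimTrail modules spawn, initialize and update the ribbon between two bone or socket indices, sampling the User.SkeletalMesh data interface.
# A color curve starts the translucent white trail at full scale, darkens it, and fades alpha out within roughly the first quarter of particle life.
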NiagaraSystem /BRCosmetics/Weapons/FORT_Melee/Pickaxe_FallValleyCharge/FX/NS_Pickaxe_FallValleyCharge_Trail
dependencies {usedDataInterfaces [&NiagaraDataInterfaceColorCurve], usedDynamicInputs [], usedModules [&SystemState]}
emitters {Ribbon_Side {
  emitterUpdateScript [
   &EmitterState()
   &FN_SpawnAnimTrail()
  ]
  particleSpawnScript [
   &FN_InitializeAnimTrailBoneIndecies(First Bone or Socket Index=3, Second Bone or Socket Index=2)
   &FN_InitializeAnimTrail(Color=<rgba 1 1 1 0.5>, Lifetime=0.2)
  ]
  particleUpdateScript [
   &SolveForcesAndVelocity(Clamp Velocity=false, Limit Acceleration=false)
   /SetVariables(Particles.RibbonV0RangeOverride:Vector2f=<1 0>) {isSetParameters=true}
   &FN_UpdateAnimTrail(Locked Particle RibbonWidth=1, SkeletalMesh=$User.SkeletalMesh) {module=FN_UpdateAnimTrail001}
   &ScaleColor(
    Linear Color Curve=DataInterface() {propertyValues {
      AlphaCurve {keys [<at 0 1 RCIM_Cubic>,<at 0.2521 0>]}
      BlueCurve {keys [{arriveTangent=-3.828, interpMode=RCIM_Cubic, leaveTangent=-3.828, tangentMode=RCTM_User, time=0, value=1}, <at 1 0.6927 RCIM_Cubic>]}
      GreenCurve {keys [{arriveTangent=-3.828, interpMode=RCIM_Cubic, leaveTangent=-3.828, tangentMode=RCTM_User, time=0, value=1}, <at 1 0.6927 RCIM_Cubic>]}
      RedCurve {keys [{arriveTangent=-3.828, interpMode=RCIM_Cubic, leaveTangent=-3.828, tangentMode=RCTM_User, time=0, value=1}, <at 1 0.6927 RCIM_Cubic>]}
     }}
    Scale Mode=::NewEnumerator2
   )
  ]
  rendererClasses [&NiagaraRibbonRendererProperties]
  renderers [{rendererClass=&NiagaraRibbonRendererProperties}]
 }}
summary {
 systemName=NS_Pickaxe_FallValleyCharge_Trail
 emitters [{emitterName=Ribbon_Side, rendererClasses [&NiagaraRibbonRendererProperties]}]
 userVariables [
  User.bIsFrontEndPreview {type=NiagaraBool}
  User.PlayerSpeed {type=NiagaraFloat, defaultValue=0, description=""}
  User.RibbonIDOffset {type=NiagaraInt32}
  User.SkeletalMesh {
   type=NiagaraDataInterfaceSkeletalMesh
   defaultValue {
    struct=NiagaraExt_VariableValue_DataInterface
    value {dataInterface=&self:NiagaraDataInterfaceSkeletalMesh_0}
   }
  }
  User.VariantIndex {type=NiagaraFloat, defaultValue=0, description=""}
 ]
}
