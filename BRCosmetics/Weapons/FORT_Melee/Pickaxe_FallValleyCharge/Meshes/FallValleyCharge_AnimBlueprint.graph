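# Anim Blueprint for the FallValleyCharge pickaxe that adds springy secondary motion to its eight dyn_umbrella bones.
# The input pose runs through chained AnimDynamics nodes, each simulated relative to dyn_umbrella_base with angular springs and tight limits.
# The update event drops alpha_acc_sc, their X acceleration scale, while jumping and interpolates it back otherwise.
# It also computes alpha_run and gravity from jump/fall state, but the AnimGraph uses neither.
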AnimBlueprint /BRCosmetics/Weapons/FORT_Melee/Pickaxe_FallValleyCharge/Meshes/FallValleyCharge_AnimBlueprint
target_skeleton=&/FallValleyCharge_Axe_Skeleton
graphs [
 AnimGraph {nodes [
   n1 = Get alpha_acc_sc() -> (alpha_acc_sc:double)
   n2 = Anim Dynamics - Bone: dyn_umbrella_1!(
    ComponentPose=Local To Component(LocalPose=Input Pose())
    Alpha=0.8
    GravityScale:float=0
    ComponentLinearAccScale_X:double=@n1
    ComponentLinearAccScale_Y:double=0.1
    ComponentLinearAccScale_Z:double=0.1
    ExternalForce="0,0,0"
   ) -> (Pose) {
    alpha=0.8
    angularBiasOverride=0.98
    angularDampingOverride=0.98
    angularSpringConstant=270
    bAngularSpring=true
    bOverrideAngularBias=true
    bOverrideAngularDamping=true
    gravityScale=0
    simulationSpace=BoneRelative
    alphaBoolBlend {blendOption=Linear}
    alphaScaleBiasClamp {bInterpResult=true}
    boundBone {boneName=dyn_umbrella_1}
    componentLinearAccScale=<0.05 0.1 0.1>
    componentLinearVelScale=<0.05 0.1 0.1>
    gravityOverride {z=-10}
    physicsBodyDefinitions [{
      boundBone {boneName=dyn_umbrella_1}
      boxExtents=<12 8 8>
      constraintSetup {twistAxis=AxisZ, angularLimitsMax=<10 0 0>, angularLimitsMin=<-10 0 0>, angularTarget=<0 0 0>}
      localJointOffset=<0 5 0>
     }]
    relativeSpaceBone {boneName=dyn_umbrella_base}
   }
   n3 = Anim Dynamics - Bone: dyn_umbrella_2!(
    ComponentPose=@n2
    Alpha=0.8
    GravityScale:float=0
    ComponentLinearAccScale_X:double=@n1
    ComponentLinearAccScale_Y:double=0.1
    ComponentLinearAccScale_Z:double=0.1
    ExternalForce="0,0,0"
   ) -> (Pose) {
    alpha=0.8
    angularBiasOverride=0.98
    angularDampingOverride=0.98
    angularSpringConstant=270
    bAngularSpring=true
    bOverrideAngularBias=true
    bOverrideAngularDamping=true
    gravityScale=0
    simulationSpace=BoneRelative
    alphaBoolBlend {blendOption=Linear}
    alphaScaleBiasClamp {bInterpResult=true}
    boundBone {boneName=dyn_umbrella_2}
    componentLinearAccScale=<0.05 0.1 0.1>
    componentLinearVelScale=<0.05 0.1 0.1>
    gravityOverride {z=-10}
    physicsBodyDefinitions [{
      boundBone {boneName=dyn_umbrella_2}
      boxExtents=<12 8 8>
      constraintSetup {twistAxis=AxisZ, angularLimitsMax=<10 0 0>, angularLimitsMin=<-10 0 0>, angularTarget=<0 0 0>}
      localJointOffset=<0 5 0>
     }]
    relativeSpaceBone {boneName=dyn_umbrella_base}
   }
   n4 = Anim Dynamics - Bone: dyn_umbrella_3!(
    ComponentPose=@n3
    Alpha=0.8
    GravityScale:float=0
    ComponentLinearAccScale_X:double=@n1
    ComponentLinearAccScale_Y:double=0.1
    ComponentLinearAccScale_Z:double=0.1
    ExternalForce="0,0,0"
   ) -> (Pose) {
    alpha=0.8
    angularBiasOverride=0.98
    angularDampingOverride=0.98
    angularSpringConstant=270
    bAngularSpring=true
    bOverrideAngularBias=true
    bOverrideAngularDamping=true
    gravityScale=0
    simulationSpace=BoneRelative
    alphaBoolBlend {blendOption=Linear}
    alphaScaleBiasClamp {bInterpResult=true}
    boundBone {boneName=dyn_umbrella_3}
    componentLinearAccScale=<0.05 0.1 0.1>
    componentLinearVelScale=<0.05 0.1 0.1>
    gravityOverride {z=-10}
    physicsBodyDefinitions [{
      boundBone {boneName=dyn_umbrella_3}
      boxExtents=<12 8 8>
      constraintSetup {twistAxis=AxisZ, angularLimitsMax=<10 0 0>, angularLimitsMin=<-10 0 0>, angularTarget=<0 0 0>}
      localJointOffset=<0 5 0>
     }]
    relativeSpaceBone {boneName=dyn_umbrella_base}
   }
   n5 = Anim Dynamics - Bone: dyn_umbrella_4!(
    ComponentPose=@n4
    Alpha=0.8
    GravityScale:float=0
    ComponentLinearAccScale_X:double=@n1
    ComponentLinearAccScale_Y:double=0.1
    ComponentLinearAccScale_Z:double=0.1
    ExternalForce="0,0,0"
   ) -> (Pose) {
    alpha=0.8
    angularBiasOverride=0.98
    angularDampingOverride=0.98
    angularSpringConstant=270
    bAngularSpring=true
    bOverrideAngularBias=true
    bOverrideAngularDamping=true
    gravityScale=0
    simulationSpace=BoneRelative
    alphaBoolBlend {blendOption=Linear}
    alphaScaleBiasClamp {bInterpResult=true}
    boundBone {boneName=dyn_umbrella_4}
    componentLinearAccScale=<0.05 0.1 0.1>
    componentLinearVelScale=<0.05 0.1 0.1>
    gravityOverride {z=-10}
    physicsBodyDefinitions [{
      boundBone {boneName=dyn_umbrella_4}
      boxExtents=<12 8 8>
      constraintSetup {twistAxis=AxisZ, angularLimitsMax=<10 0 0>, angularLimitsMin=<-10 0 0>, angularTarget=<0 0 0>}
      localJointOffset=<0 5 0>
     }]
    relativeSpaceBone {boneName=dyn_umbrella_base}
   }
   n6 = Anim Dynamics - Bone: dyn_umbrella_5!(
    ComponentPose=@n5
    Alpha=0.8
    GravityScale:float=0
    ComponentLinearAccScale_X:double=@n1
    ComponentLinearAccScale_Y:double=0.1
    ComponentLinearAccScale_Z:double=0.1
    ExternalForce="0,0,0"
   ) -> (Pose) {
    alpha=0.8
    angularBiasOverride=0.98
    angularDampingOverride=0.98
    angularSpringConstant=270
    bAngularSpring=true
    bOverrideAngularBias=true
    bOverrideAngularDamping=true
    gravityScale=0
    simulationSpace=BoneRelative
    alphaBoolBlend {blendOption=Linear}
    alphaScaleBiasClamp {bInterpResult=true}
    boundBone {boneName=dyn_umbrella_5}
    componentLinearAccScale=<0.05 0.1 0.1>
    componentLinearVelScale=<0.05 0.1 0.1>
    gravityOverride {z=-10}
    physicsBodyDefinitions [{
      boundBone {boneName=dyn_umbrella_5}
      boxExtents=<12 8 8>
      constraintSetup {twistAxis=AxisZ, angularLimitsMax=<10 0 0>, angularLimitsMin=<-10 0 0>, angularTarget=<0 0 0>}
      localJointOffset=<0 5 0>
     }]
    relativeSpaceBone {boneName=dyn_umbrella_base}
   }
   n7 = Anim Dynamics - Bone: dyn_umbrella_6!(
    ComponentPose=@n6
    Alpha=0.8
    GravityScale:float=0
    ComponentLinearAccScale_X:double=@n1
    ComponentLinearAccScale_Y:double=0.1
    ComponentLinearAccScale_Z:double=0.1
    ExternalForce="0,0,0"
   ) -> (Pose) {
    alpha=0.8
    angularBiasOverride=0.98
    angularDampingOverride=0.98
    angularSpringConstant=270
    bAngularSpring=true
    bOverrideAngularBias=true
    bOverrideAngularDamping=true
    gravityScale=0
    simulationSpace=BoneRelative
    alphaBoolBlend {blendOption=Linear}
    alphaScaleBiasClamp {bInterpResult=true}
    boundBone {boneName=dyn_umbrella_6}
    componentLinearAccScale=<0.05 0.1 0.1>
    componentLinearVelScale=<0.05 0.1 0.1>
    gravityOverride {z=-10}
    physicsBodyDefinitions [{
      boundBone {boneName=dyn_umbrella_6}
      boxExtents=<12 8 8>
      constraintSetup {twistAxis=AxisZ, angularLimitsMax=<10 0 0>, angularLimitsMin=<-10 0 0>, angularTarget=<0 0 0>}
      localJointOffset=<0 5 0>
     }]
    relativeSpaceBone {boneName=dyn_umbrella_base}
   }
   n8 = Anim Dynamics - Bone: dyn_umbrella_7!(
    ComponentPose=@n7
    Alpha=0.8
    GravityScale:float=0
    ComponentLinearAccScale_X:double=@n1
    ComponentLinearAccScale_Y:double=0.1
    ComponentLinearAccScale_Z:double=0.1
    ExternalForce="0,0,0"
   ) -> (Pose) {
    alpha=0.8
    angularBiasOverride=0.98
    angularDampingOverride=0.98
    angularSpringConstant=270
    bAngularSpring=true
    bOverrideAngularBias=true
    bOverrideAngularDamping=true
    gravityScale=0
    simulationSpace=BoneRelative
    alphaBoolBlend {blendOption=Linear}
    alphaScaleBiasClamp {bInterpResult=true}
    boundBone {boneName=dyn_umbrella_7}
    componentLinearAccScale=<0.05 0.1 0.1>
    componentLinearVelScale=<0.05 0.1 0.1>
    gravityOverride {z=-10}
    physicsBodyDefinitions [{
      boundBone {boneName=dyn_umbrella_7}
      boxExtents=<12 8 8>
      constraintSetup {twistAxis=AxisZ, angularLimitsMax=<10 0 0>, angularLimitsMin=<-10 0 0>, angularTarget=<0 0 0>}
      localJointOffset=<0 5 0>
     }]
    relativeSpaceBone {boneName=dyn_umbrella_base}
   }
   n9 = Anim Dynamics - Bone: dyn_umbrella_8!(
    ComponentPose=@n8
    Alpha=0.8
    GravityScale:float=0
    ComponentLinearAccScale_X:double=@n1
    ComponentLinearAccScale_Y:double=0.1
    ComponentLinearAccScale_Z:double=0.1
    ExternalForce="0,0,0"
   ) -> (Pose) {
    alpha=0.8
    angularBiasOverride=0.98
    angularDampingOverride=0.98
    angularSpringConstant=270
    bAngularSpring=true
    bOverrideAngularBias=true
    bOverrideAngularDamping=true
    gravityScale=0
    simulationSpace=BoneRelative
    alphaBoolBlend {blendOption=Linear}
    alphaScaleBiasClamp {bInterpResult=true}
    boundBone {boneName=dyn_umbrella_8}
    componentLinearAccScale=<0.05 0.1 0.1>
    componentLinearVelScale=<0.05 0.1 0.1>
    gravityOverride {z=-10}
    physicsBodyDefinitions [{
      boundBone {boneName=dyn_umbrella_8}
      boxExtents=<12 8 8>
      constraintSetup {twistAxis=AxisZ, angularLimitsMax=<10 0 0>, angularLimitsMin=<-10 0 0>, angularTarget=<0 0 0>}
      localJointOffset=<0 5 0>
     }]
    relativeSpaceBone {boneName=dyn_umbrella_base}
   }
   Output Pose(Result=Component To Local(ComponentPose=@n9)) {layerGroup=None}
  ]}
 EventGraph {nodes [
   n1 = Event BlueprintUpdateAnimation()
   > Set deltaTime(deltaTime:double=@n1.DeltaTimeX) -> (Output_Get:double)
   > n2 = Sequence!(execute) -> (then_0, then_1, then_2:Exec)
   > n3 = Branch(Condition=Get bIsJumping())
   > Set alpha_acc_sc(alpha_acc_sc:double=0.01) -> (Output_Get:double)
   Set alpha_acc_sc(
    execute=@n3.else
    alpha_acc_sc=FInterpTo(Current=Get alpha_acc_sc(), Target=0.8, DeltaTime=Get deltaTime() -> (deltaTime:double), InterpSpeed=20)
   )
   n4 = Branch(execute=@n2.then_1, Condition=Get bIsJumping())
   > Set alpha_run(alpha_run:double=0.3) -> (Output_Get:double)
   Set alpha_run(execute=@n4.else, alpha_run=FInterpTo(Current=Get alpha_run() -> (alpha_run:double), Target=1, DeltaTime=Get deltaTime(), InterpSpeed=20))
   n5 = Branch(execute=@n2.then_2, Condition=OR Boolean(A=Get bIsJumping(), B=Get bIsFalling()))
   > Set gravity(gravity:double=FInterpTo(Current=Get gravity() -> (gravity:double), Target=0.1, DeltaTime=Get deltaTime(), InterpSpeed=20)) -> (Output_Get:double)
   Set gravity(execute=@n5.else, gravity=FInterpTo(Current=Get gravity(), Target=0, DeltaTime=Get deltaTime(), InterpSpeed=25))
   TryGetPawnOwner()
   FInterpTo(Current=Get alpha_acc_sc(), Target=0, DeltaTime=Get deltaTime(), InterpSpeed=1)
  ]}
]
variables [
 alpha_acc_sc {type=number}
 alpha_run {type=number}
 deltaTime {type=number}
 gravity {type=number}
]
